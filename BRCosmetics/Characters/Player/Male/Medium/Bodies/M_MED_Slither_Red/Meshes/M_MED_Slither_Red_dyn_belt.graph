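# Physics asset for the dynamic belt on the medium male Slither Red player body.
# Small master capsules root a four-link right chain and a single-link left segment to the pelvis, with tight swing and twist limits along the links.
# Pelvis, thigh-twist and large front-skirt capsules supply body collision volumes.
# Master and body-volume constraints leave rotation free, and capsule names reused from skirt bones suggest copying from a skirt setup.
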PhysicsAsset /BRCosmetics/Characters/Player/Male/Medium/Bodies/M_MED_Slither_Red/Meshes/M_MED_Slither_Red_dyn_belt
bodies {
 dyn_belt_MSTR [{shapeType=Capsule, length=0.505, radius=0.505, shapeName=dyn_skirt_bk_r_MSTR_capsule, rotation=<rot 0 -90 0>}]
 dyn_belt_l_01 [{shapeType=Capsule, length=2.75, radius=2, shapeName=dyn_skirt_bk_l_01_capsule, center=<4 0 0>, rotation=<rot 90 0 0>}]
 dyn_belt_l_MSTR [{shapeType=Capsule, length=0.505, radius=0.505, shapeName=dyn_skirt_bk_r_MSTR_capsule, rotation=<rot 0 -90 0>}]
 dyn_belt_r_01 [{shapeType=Capsule, length=3.25, radius=1.75, shapeName=dyn_skirt_bk_l_01_capsule, center=<5 0 0>, rotation=<rot 90 0 0>}]
 dyn_belt_r_02 [{shapeType=Capsule, length=2.75, radius=2, shapeName=dyn_skirt_bk_l_01_capsule, center=<5 0 0>, rotation=<rot 90 0 0>}]
 dyn_belt_r_03 [{shapeType=Capsule, length=3, radius=2, shapeName=dyn_skirt_bk_l_02_capsule, center=<5 0 0>, rotation=<rot 90 0 0>}]
 dyn_belt_r_04 [{shapeType=Capsule, length=4, radius=2, shapeName=dyn_skirt_bk_l_02_capsule, center=<5 0 0>, rotation=<rot 90 0 0>}]
 dyn_belt_r_MSTR [{shapeType=Capsule, length=0.505, radius=0.505, shapeName=dyn_skirt_bk_r_MSTR_capsule, rotation=<rot 0 -90 0>}]
 dyn_skirt_ft_collision [{shapeType=Capsule, length=18.5, radius=32.7732, shapeName=pelvis_capsule, center=<8.1416 -23.8888 -0.0004>}]
 pelvis [{shapeType=Capsule, length=31.5, radius=14.2732, shapeName=pelvis_capsule, center=<-0.575 -5.6905 -0.0004>}]
 thigh_twist_01_l [{
   shapeType=Capsule
   length=34.682
   radius=9.7357
   shapeName=thigh_twist_01_l_capsule
   center=<1.6886 -1.4454 2.2502>
   rotation=<rot 86.6465 1.163 -1.466>
  }]
 thigh_twist_01_r [{
   shapeType=Capsule
   length=34.682
   radius=9.7357
   shapeName=thigh_twist_01_l_capsule
   center=<-1.6886 1.4454 -2.2502>
   rotation=<rot 86.6466 1.163 178.534>
  }]
}
constraints [
 {bone1Name=dyn_belt_r_01, bone2Name=dyn_belt_r_MSTR, swing1LimitDegrees=20, swing2LimitDegrees=35, twistLimitDegrees=5}
 {bone1Name=dyn_belt_r_02, bone2Name=dyn_belt_r_01, swing1LimitDegrees=20, swing2LimitDegrees=35, twistLimitDegrees=5}
 {bone1Name=dyn_belt_r_03, bone2Name=dyn_belt_r_02, swing1LimitDegrees=30, swing2LimitDegrees=35, twistLimitDegrees=5}
 {bone1Name=dyn_belt_l_01, bone2Name=dyn_belt_l_MSTR, swing1LimitDegrees=20, swing2LimitDegrees=15, twistLimitDegrees=2}
 {
  bone1Name=dyn_belt_r_MSTR
  bone2Name=dyn_belt_MSTR
  swing1LimitDegrees=45
  swing1Motion=Free
  swing2LimitDegrees=45
  swing2Motion=Free
  twistLimitDegrees=45
  twistMotion=Free
 }
 {
  bone1Name=dyn_belt_l_MSTR
  bone2Name=dyn_belt_MSTR
  swing1LimitDegrees=45
  swing1Motion=Free
  swing2LimitDegrees=45
  swing2Motion=Free
  twistLimitDegrees=45
  twistMotion=Free
 }
 {bone1Name=dyn_belt_r_04, bone2Name=dyn_belt_r_03, swing1LimitDegrees=30, swing2LimitDegrees=35, twistLimitDegrees=5}
 {
  bone1Name=thigh_twist_01_r
  bone2Name=pelvis
  swing1LimitDegrees=45
  swing1Motion=Free
  swing2LimitDegrees=45
  swing2Motion=Free
  twistLimitDegrees=45
  twistMotion=Free
 }
 {
  bone1Name=dyn_skirt_ft_collision
  bone2Name=pelvis
  swing1LimitDegrees=45
  swing1Motion=Free
  swing2LimitDegrees=45
  swing2Motion=Free
  twistLimitDegrees=45
  twistMotion=Free
 }
 {
  bone1Name=dyn_belt_MSTR
  bone2Name=pelvis
  swing1LimitDegrees=45
  swing1Motion=Free
  swing2LimitDegrees=45
  swing2Motion=Free
  twistLimitDegrees=45
  twistMotion=Free
 }
 {
  bone1Name=thigh_twist_01_l
  bone2Name=pelvis
  swing1LimitDegrees=45
  swing1Motion=Free
  swing2LimitDegrees=45
  swing2Motion=Free
  twistLimitDegrees=45
  twistMotion=Free
 }
]
